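; Spicy Berry bush building actor: players eat or gather its berries while loot remains.
; Server enables loot from playlist data; berry count comes from a scalable float.
; Eating applies a gameplay effect and event; gathering grants a loot-table item, or drops one and clears loot when inventory is full.
; Picks remove instanced berry meshes with FX, death drops remaining berries, and HasLoot replication updates client visibility and bounce.
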(event EventBeginPlay
  (CallFunction|ForceSetMeshCollision)
  (if (Networking|IsServer)
    (Class|FortAsyncActionCurrentPlaylistReady|PlaylistDataReadyAsync
      (:OnReady
        (Variables|Default|SetHasLoot (Class|AthenaFunctionLibrary|RowtoBool (Variables|Default|GetBerriesEnabled)))
        (CallFunction|InitializeBerryCount)))))

(event Interaction|EventBlueprintOnInteract
  (params (InteractingPawn :type Fort%20Pawn%20Object%20Reference) (InteractionBeingAttempted :type EInteractionBeingAttempted%20Enum :default "FirstInteraction"))
  (|CallParentFunction InteractingPawn InteractionBeingAttempted)
  (if (Variables|Default|GetHasLoot)
    (bind _output_get (Variables|Default|SetLastInteractingPawn InteractingPawn))
    (switch Utilities|FlowControl|Switch|SwitchonEInteractionBeingAttempted InteractionBeingAttempted
      (:FirstInteraction
        (CallFunction|ConsumeBerry))
      (:SecondInteraction
        (bind _returnvalue (CallFunction|HasOpenInventorySlot _output_get))
        (if _returnvalue
          (CallFunction|GiveItemtoInventory)
          (CallFunction|OnInteractAudio)
          (else
            (CallFunction|SpawnLoot)))))))

(event Custom|GiveItemToInventory
  (bind _asfort_player_controller (Utilities|Casting|CastToFortPlayerController (Pawn|GetController (Variables|Default|GetLastInteractingPawn))))
  ;[1 1 1] Get the item definition
  (bind _item_definition (CallFunction|GetItemtoGrant))
  (bind _asfort_world_item_definition (Utilities|Casting|CastToFortWorldItemDefinition _item_definition))
  (Inventory|GiveItemtoInventoryOwner _asfort_player_controller _asfort_world_item_definition)
  (CallFunction|DecrementBerryCounter))

(event Custom|SpawnLoot
  (Variables|Default|SetHasLoot false)
  (Class|AthenaFunctionLibrary|SpawnLootfromTable (Variables|Default|GetLootTierGroupName) 1 -1 (Transformation|GetActorLocation) "0, 0,10.000000" 1 true)
  (CallFunction|DecrementBerryCounter))

(event EventOnDeathServer
  (params (Damage :type Float%20%28single-precision%29) (DamageTags :type Gameplay%20Tag%20Container%20Structure%20%28by%20ref%29) (Momentum :type Vector :default "0, 0, 0") (HitInfo :type Hit%20Result%20Structure%20%28by%20ref%29) (InstigatedBy :type Controller%20Object%20Reference) (DamageCauser :type Actor%20Object%20Reference) (EffectContext :type Gameplay%20Effect%20Context%20Handle%20Structure))
  ;[1 1 1] Handle death
  (if (Variables|Default|GethasLoot)
    (for _ (range 1 (+ (Variables|Default|GetBerriesRemaining) 1))
      (Class|AthenaFunctionLibrary|SpawnLootfromTable (Variables|Default|GetLootTierGroupName) 1 -1 (Transformation|GetActorLocation) "0, 0,10.000000" 0 true))))

(event EventOnDeathPlayEffects
  (params (Damage :type Float%20%28single-precision%29) (DamageTags :type Gameplay%20Tag%20Container%20Structure%20%28by%20ref%29) (Momentum :type Vector :default "0, 0, 0") (HitInfo :type Hit%20Result%20Structure%20%28by%20ref%29) (InstigatedBy :type Fort%20Pawn%20Object%20Reference) (DamageCauser :type Actor%20Object%20Reference) (EffectContext :type Gameplay%20Effect%20Context%20Handle%20Structure))
  ;[1 1 1] Handle death
  (Development|SetHiddeninGame (Variables|Debug|GetStaticMeshComponent) true true))

(event Custom|OnInteractAudio
  (bind _berries_wid (Variables|Default|GetBerries_WID)
    (:then
      (bind _asfort_weapon_item_definition (Utilities|Casting|CastToFortWeaponItemDefinition _berries_wid))
      (bind _itemdata (Class|ItemDefinitionBase|GetItemData _asfort_weapon_item_definition)
        (:Found
          (if (Utilities|IsValidSoftObjectReference (.PickupSound (Utilities|Struct|BreakFortnitePickupItemData _itemdata)))
            (Utilities|AsyncLoadAsset (.PickupSound (Utilities|Struct|BreakFortnitePickupItemData _itemdata))
              (:Completed
                (if (Pawn|IsLocallyViewed (Variables|Default|GetLastInteractingPawn))
                  (Audio|PlaySound2D (Utilities|ResolveSoftReference (.PickupSound (Utilities|Struct|BreakFortnitePickupItemData _itemdata)))))))))))
    (:else)))

(event Building|EventOnBuildingActorInitialized
  (params (InitializationReason :type EFortBuildingInitializationReason%20Enum :default "StaticallyPlaced") (BuildingPersistentState :type EFortBuildingPersistentState%20Enum :default "Default"))
  (|CallParentFunction InitializationReason BuildingPersistentState)
  (CallFunction|InitializeBerryCount))

(event Custom|ConsumeBerry
  (bind _lastinteractingpawn (Variables|Default|GetLastInteractingPawn))
  (Environment|ApplyGlobalEnvironmentGameplayEffecttoActor _lastinteractingpawn "/SpicyBerry/Gameplay/GE_Athena_SpicyBerry_Bush_Consume_Grant.GE_Athena_SpicyBerry_Bush_Consume_Grant_C")
  (CallFunction|DecrementBerryCounter)
  (Ability|SendGameplayEventtoActor _lastinteractingpawn "(TagName=\"Athena.event.gameplay.ConsumedBerry\")"))

(event Custom|UpdateBerries
  (CallFunction|DecrementBerryInstances))

(fn BlueprintCanInteract
  (params (InteractingPawn :type Fort%20Pawn%20Object%20Reference) (InteractionBeingAttempted :type EInteractionBeingAttempted%20Enum :default "FirstInteraction") (InteractionType :type TInteractionType%20Enum :default "IT_NoInteraction"))
  (switch Utilities|FlowControl|Switch|SwitchonEInteractionBeingAttempted InteractionBeingAttempted
    (:FirstInteraction
      (return (Variables|Default|GethasLoot)))
    (:SecondInteraction
      (return (and (not (GameplayTags|HasAnyMatchingGameplayTags InteractingPawn (Variables|Default|GetBlockConsumeTags))) (Variables|Default|GethasLoot))))))

(fn BlueprintGetInteractionString
  (params (InteractingPawn :type Fort%20Pawn%20Object%20Reference) (InteractionBeingAttempted :type EInteractionBeingAttempted%20Enum :default "FirstInteraction"))
  (switch Utilities|FlowControl|Switch|SwitchonEInteractionBeingAttempted InteractionBeingAttempted
    (:FirstInteraction
      (return (Variables|Default|GetInteractionLabel)))
    (:SecondInteraction
      (return (Variables|Default|GetGatherLabel)))))

(fn DecrementBerryCounter
  (bind _berriesremaining (Variables|Default|GetBerriesRemaining))
  (Math|Integer|DecrementInt _berriesremaining
    (:%20%20
      (CallFunction|UpdateBerries)
      (if (<= _berriesremaining 0)
        (Variables|Default|SetHasLoot false)))))

(fn DecrementBerryInstances
  (bind _outinstancetransform (Components|InstancedStaticMesh|GetInstanceTransform (Variables|Default|GetHierarchicalInstancedStaticMesh) (- (Components|InstancedStaticMesh|GetInstanceCount (Variables|Default|GetHierarchicalInstancedStaticMesh)) 1) true))
  (bind (_outinstancetransform_location _outinstancetransform_rotation _outinstancetransform_scale) (Math|Transform|BreakTransform _outinstancetransform))
  (Effects|SpawnEmitteratLocation (Variables|VFX|GetBerryPickedOffVineFX) _outinstancetransform_location _outinstancetransform_rotation _outinstancetransform_scale)
  (Components|InstancedStaticMesh|RemoveInstance (Variables|Default|GetHierarchicalInstancedStaticMesh) (- (Components|InstancedStaticMesh|GetInstanceCount (Variables|Default|GetHierarchicalInstancedStaticMesh)) 1)))

(fn ForceSetMeshCollision
  (bind _staticmeshcomponent (Variables|Debug|GetStaticMeshComponent))
  (Collision|SetCollisionResponsetoAllChannels _staticmeshcomponent "ECR_Ignore")
  (for _array_element (Utilities|Array|MakeArray "ECC_GameTraceChannel1" "ECC_GameTraceChannel2" "ECC_GameTraceChannel14" "ECC_GameTraceChannel16")
    (Collision|SetCollisionResponsetoChannel _staticmeshcomponent _array_element "ECR_Overlap"))
  (for _array_element_2 (Utilities|Array|MakeArray "ECC_Visibility" "ECC_GameTraceChannel3" "ECC_Destructible" "ECC_GameTraceChannel18" "ECC_Vehicle")
    (Collision|SetCollisionResponsetoChannel _staticmeshcomponent _array_element_2 "ECR_Block")))

(fn GetItemToGrant
  (bind (_outloottodrop _returnvalue) (Loot|PickLootDrops (Variables|Default|GetLootTierGroupName) 0 -1 "(GameplayTags=)"))
  (if _returnvalue
    (return (Class|FortItemFunctionLibrary|BreakFortItemEntry (Utilities|Array|Get(acopy) _outloottodrop)) (.Count (Class|FortItemFunctionLibrary|BreakFortItemEntry (Utilities|Array|Get(acopy) _outloottodrop))))))

(fn HasOpenInventorySlot
  (params (Pawn :type Pawn%20Object%20Reference))
  (bind _asfort_player_controller (Utilities|Casting|CastToFortPlayerController (Pawn|GetController Pawn))
    (:then
      (bind _returnvalue (IndexSelectionModel|GetNumItems(Message) _asfort_player_controller))
      (return (< _returnvalue 5)))
    (:CastFailed
      (return false))))

(fn InitializeBerryCount
  (Variables|Default|SetBerriesRemaining (Math|Integer|Clamp(Integer) (Math|Float|Truncate (ScalableFloat|GetValueatLevel (Variables|Default|GetBerryCount))) 1 4)))

(fn OnRep_HasLoot
  (bind _returnvalue (Transformation|GetActorLocation))
  (bind _returnvalue_2 (+ _returnvalue "0,0,90.000000"))
  (if (not (Networking|IsServer))
    (Rendering|SetVisibility (Variables|Default|GetHierarchicalInstancedStaticMesh) (Variables|Default|GethasLoot) false)
    (Utilities|IsValid (Variables|Default|GetLastInteractingPawn)
      (:Is%20Valid
        (CallFunction|StartBounceAnimation (Utilities|Time|GetTimeSeconds) 1.2e+02 (Math|Vector|Normalize (- _returnvalue (Transformation|GetActorLocation (Variables|Default|GetLastInteractingPawn)))) _returnvalue_2 "Interact" true))
      (:Is%20Not%20Valid
        (CallFunction|StartBounceAnimation (Utilities|Time|GetTimeSeconds) 1.2e+02 (Math|Vector|MakeVector 0.0 0.0 1.0) _returnvalue_2 "Interact" true)))))

(fn ConstructionScript
  (|CallParentFunction))
