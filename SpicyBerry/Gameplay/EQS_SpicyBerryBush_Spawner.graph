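# Environment query that picks spawn locations for Spicy Berry bushes around the querier.
# A donut generator lays spiral-patterned points in two rings and projects them onto geometry with a channel trace.
# Tests score each point for surface flatness, filter and score it with a small overlap check on a game trace channel, and add square-root-weighted randomness to vary placement.
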EnvQuery /SpicyBerry/Gameplay/EQS_SpicyBerryBush_Spawner
options [{
  generator {
   name=EnvQueryGenerator_Donut_0
   class=EnvQueryGenerator_Donut
   properties {
    bUseSpiralPattern=true
    center=&EnvQueryContext_Querier
    innerRadius=400
    numberOfRings=2
    outerRadius=750
    pointsPerRing=5
    projectionData {
     extentX=30
     navigationFilter=&NavFilter_NPC_ExcludeAllWater_C
     postProjectionVerticalOffset=0
     projectDown=2048
     projectUp=2048
     serializedChannel=ECC_GameTraceChannel6
     traceChannel=TraceTypeQuery7
     traceMode=GeometryByChannel
    }
   }
  }
  tests [
   FortQueryTest_OnFlatSurface_0 {class=FortQueryTest_OnFlatSurface, properties {radius=50, testPurpose=Score, toleranceZ=10, traceOffsetUp=100}}
   EnvQueryTest_Overlap_0 {
    class=EnvQueryTest_Overlap
    properties {testPurpose=FilterAndScore, overlapData {extentX=5, extentY=5, extentZ=5, overlapChannel=ECC_GameTraceChannel15, shapeOffset=<0 0 0>}}
   }
   FortQueryTest_Random_0 {class=FortQueryTest_Random, properties {bUseRandomSeedForOthers=true, floatValueMax=0, scoringEquation=SquareRoot, scoringFactor=1}}
  ]
 }]
